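; Bank variant of the VividRazor vault actor, adding AI break-in interruption and a minimap icon.
; When an alert starts, the server enables a Smart Object if pausing is allowed; AI stopping the break-in pauses the manager's alert and resets music.
; The authority async-loads the icon and swaps its material by vault progress and boss visibility.
; Opening the door hides the icon, disables the Smart Object and sets debris visibility.
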(event EventHandleAlertStarted
  ; Alert started
  (|CallParentFunction)
  (if (ScalableFloat|GetValueasBool (Variables|ScalableFloats|GetROW_Pause_Enabled))
    (if (Networking|IsServer)
      (CallFunction|SmartObjectActivation true) ; Enable Smart Object so AI can stop the vault
    )))

(event EventBeginPlay
  (|CallParentFunction)
  (CallFunction|SmartObjectActivation false)
  (Networking|SwitchHasAuthority
    (:Authority
      (if (not (Building|IsPreviewBuildingActor))
        (CallFunction|SpawnorDestroyMapIcon true)))))

(event EventEndPlay
  (params (EndPlayReason :type EEndPlayReason%20Enum :default "Destroyed"))
  (|CallParentFunction EndPlayReason)
  (CallFunction|SpawnorDestroyMapIcon false))

(event Custom|AIStoppedVaultBreakIn
  (Development|PrintString "AI MANAGED TO STOP THE VAULT BREAKIN!")
  (CallFunction|SmartObjectActivation false)
  (Class|BGAVividRazorVaultManager|AlertPause (Variables|Default|GetBGA_VaultManager))
  (Variables|Default|SetMusic%20Stage "NewEnumerator0"))

(event EventResetVisuals
  (|CallParentFunction)
  (Rendering|SetVisibility (Variables|Default|GetSM_Gameplay_VaultDoorDebris_B) false true)
  (CallFunction|SpawnorDestroyMapIcon (CallFunction|IsMapIconEnabled))
  (Variables|Default|SetIsVaultInProgress false)
  (CallFunction|RefreshMinimapIcon))

(event EventHandleAlertPaused
  (|CallParentFunction)
  (Variables|Default|SetIsVaultInProgress false)
  (CallFunction|RefreshMinimapIcon))

(event Custom|LoadAndSpawnMapIcon
  (Utilities|FlowControl|DoOnce false
    (:Completed
      (Utilities|AsyncLoadClassAsset (Variables|Default|GetMapIconClassToSpawn)
        (:Completed
          (Utilities|FlowControl|Sequence
            (:then_0
              _)
            (:then_1
              _))
          (bind _returnvalue (Game|SpawnActorfromClass (Utilities|ResolveSoftReference (Variables|Default|GetMapIconClassToSpawn)) (Math|Conversions|ToTransform(Vector) (Transformation|GetActorLocation)) "AlwaysSpawn" "MultiplyWithRoot" 0 0))
          (Variables|Default|SetMapIconActor _returnvalue)
          (CallFunction|RefreshMinimapIcon))))))

(fn IsMapIconEnabled
  (return true))

(fn OnBossSetEvent
  (params (ShowBossOnMinimap :type Boolean))
  (Variables|Default|SetShowBossOnMinimap ShowBossOnMinimap)
  (CallFunction|RefreshMinimapIcon))

(fn OnDoorOpen
  (|CallParentFunction)
  (bind _mapiconactor (Variables|Default|GetMapIconActor)
    (:then
      (Class|BPIVividRazorVaultMapIconInterface|SetMapIconVisible(Message) _mapiconactor false))
    (:else
      _))
  (CallFunction|SmartObjectActivation false) ; Disable Smart Object so AI don't try to keep going at it.
  (Rendering|SetVisibility (Variables|Default|GetSM_Gameplay_VaultDoorDebris_A) (Variables|Default|GetShowDebris_Behind) true)
  (Rendering|SetVisibility (Variables|Default|GetSM_Gameplay_VaultDoorDebris_B) (Variables|Default|GetShowDebris_Front) true))

(fn RefreshMinimapIcon
  (Networking|SwitchHasAuthority
    (:Authority
      (Utilities|IsValid (Variables|Default|GetMapIconActor)
        (:Is%20Valid
          (if (Variables|Default|GetShowBossOnMinimap)
            (Class|BPIVividRazorVaultMapIconInterface|SetMapIcon(Message) (Variables|Default|GetMapIconActor) (select (Variables|Default|GetIsVaultInProgress) "/VividRazorVaultGameplay/UI/Materials/MI_UI_Minimap_Vault_InProgress_WithBoss.MI_UI_Minimap_Vault_InProgress_WithBoss" "/VividRazorVaultGameplay/UI/Materials/MI_UI_Minimap_Vault_WithBoss.MI_UI_Minimap_Vault_WithBoss"))
            (else
              (Class|BPIVividRazorVaultMapIconInterface|SetMapIcon(Message) (Variables|Default|GetMapIconActor) (select (Variables|Default|GetIsVaultInProgress) "/VividRazorVaultGameplay/UI/Materials/MI_UI_Minimap_Vault_InProgress.MI_UI_Minimap_Vault_InProgress" "/VividRazorVaultGameplay/UI/Materials/MI_UI_Minimap_Vault.MI_UI_Minimap_Vault")))))))))

(fn RequestVaultStart
  (params (Payload :type Struct%20Vivid%20Razor%20Core%20Vault%20Start%20Payload%20Structure))
  (bind _preservetriggeringactor (|CallParentFunction Payload))
  (Variables|Default|SetIsVaultInProgress true)
  (CallFunction|RefreshMinimapIcon)
  (return _preservetriggeringactor))

(fn SmartObjectActivation
  (params (EnableSmartObject :type Boolean))
  (if (not (SmartObject|SetSmartObjectEnabled(defaultreason:Gameplay) (Variables|Default|GetFortSmartObject) EnableSmartObject :!class /Script/SmartObjectsModule.SmartObjectComponent))
    (Fortnite|FortShippingLog self "Trying to Enable/Disable Smart Object Component, but Status did not change!" true))
  (return))

(fn SpawnOrDestroyMapIcon
  (params (Spawn :type Boolean))
  (bind _mapiconactor (Variables|Default|GetMapIconActor))
  (Networking|SwitchHasAuthority
    (:Authority
      (if (Editor|InEditor)
        _
        (elif Spawn
          (Utilities|IsValid (Variables|Default|GetMapIconActor)
            (:Is%20Valid
              (Class|BPIVividRazorVaultMapIconInterface|SetMapIconVisible(Message) (Variables|Default|GetMapIconActor) true))
            (:Is%20Not%20Valid
              (CallFunction|LoadandSpawnMapIcon) ; Async load and create icon if it did not exist
            ))
          (else
            (Utilities|IsValid _mapiconactor
              (:Is%20Valid
                (Actor|DestroyActor _mapiconactor) ; Destroy icon
              ))))))))

(fn ConstructionScript
  (|CallParentFunction))
