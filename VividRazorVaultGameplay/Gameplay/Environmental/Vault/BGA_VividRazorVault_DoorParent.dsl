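; Parent Blueprint for a vault door that players breach by placing thermite, then damaging weakpoints until it opens.
; Interaction checks for thermite and starts the vault manager; event-router payloads matched by location tag drive alert and opening.
; Weakpoints shed covers after a delay, take multiplier-scaled player damage plus staged, drill-aware damage over time, and report progress to the manager.
; Per-shape dynamic material parameters, Niagara, audio and music stages provide feedback.
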(event Custom|HandleAlertStarted
  ; Alert Started
  ;;[0.8902 0.8958 0.06136 0.5] TODO handle resume/pause
  (Variables|Interaction|SetbAllowInteract false)
  (Variables|Default|SetbIsActive true)
  (Visuals|UpdateThermiteMaterials true)
  (Components|Activation|Activate (Variables|Default|GetNS_Thermite_for_Vault_Proto) false)
  (if (not (Networking|IsServer))
    (CallFunction|ThermiteAudio true)
    (Visuals|UpdateLights true)
    (Variables|Default|SetMusic%20Stage "NewEnumerator1")))

(event EventOnWorldReady
  (|CallParentFunction)
  (Utilities|TryGetEventRouter
    (:Then
      (CallFunction|ListenforVaultOpened)
      (CallFunction|ListenforBossSetEvent)
      (bind (_asynctaskproxy _payload) (GameplayEventRouter|StartListeningToEvent (Variables|Default|GetVaultEventRouter) "/VividRazorCoreGameplay/Gameplay/Structs/Struct_VividRazorCore_Vault_Alert_Payload.Struct_VividRazorCore_Vault_Alert_Payload")
        (:OnEventReceived
          (if (CallFunction|IsVaultAlertPayloadMine _payload)
            (CallFunction|HandleAlertStarted)))))))

(event EventBeginPlay
  ; Begin Play
  (|CallParentFunction)
  (Visuals|InitializeDynamicMaterial)
  (Visuals|UpdateLights false)
  (CallFunction|ThermiteAudio false)
  (CallFunction|InitializeMusic))

(event Custom|ListenForVaultOpened
  ; Event: Vault Opened, only trigger when door opens, visuals should be updated in OnDoorOpen
  (bind (_asynctaskproxy _payload) (GameplayEventRouter|StartListeningToStatefulEvent (Variables|Default|GetVaultEventRouter) "/VividRazorCoreGameplay/Gameplay/Structs/Struct_VividRazorCore_Vault_Opened_Payload.Struct_VividRazorCore_Vault_Opened_Payload")
    (:then
      (Variables|Default|SetVaultOpenedListenTask _asynctaskproxy))
    (:OnEventReceived
      (if (CallFunction|IsVaultOpenedPayloadMine _payload)
        (GameplayEventRouter|StopListeningtoStatefulEvent (Variables|Default|GetVaultOpenedListenTask))
        (CallFunction|SetDoorOpened)
        (CallFunction|ThermiteAudio false)))))

(event EventHandleDoorOpened
  (CallFunction|OnDoorOpen)
  (Variables|Default|SetMusic%20Stage "NewEnumerator2")
  ; In case door streamed in after opened
  (Networking|CanExecuteCosmeticEvents
    (:True
      (Class|BGAVividRazorVaultManager|BroadcastVaultOpenedtoAllClients (Variables|Default|GetBGA_VaultManager))
      (Utilities|FlowControl|Delay 1.0)
      (Variables|Default|SetMusic%20Stage "NewEnumerator3"))))

(event Custom|ThermiteAudio
  (params (bPlay :type Boolean))
  (Networking|CanExecuteCosmeticEvents
    (:True
      (if bPlay
        (Variables|Default|GetThermiteAudioComponent
          (:then)
          (:else
            (bind _returnvalue (Audio|SpawnSoundatLocation (Variables|Default|GetThermiteSound) (Transformation|GetWorldLocation (Variables|Default|GetThermite))))
            (Variables|Default|SetThermiteAudioComponent _returnvalue)))
        (else
          (bind _thermiteaudiocomponent (Variables|Default|GetThermiteAudioComponent)
            (:then
              (Audio|Components|Audio|FadeOut _thermiteaudiocomponent 0.5 0.0 "Linear" :!class /Script/Engine.AudioComponent))
            (:else)))))))

(event Custom|ThermiteAudioReliable
  (params (bPlay :type Boolean))
  (Networking|CanExecuteCosmeticEvents
    (:True
      (if bPlay
        (Variables|Default|GetThermiteAudioComponent
          (:then)
          (:else
            (bind _returnvalue (Audio|SpawnSoundatLocation (Variables|Default|GetThermiteSound) (Transformation|GetWorldLocation (Variables|Default|GetThermite))))
            (Variables|Default|SetThermiteAudioComponent _returnvalue)))
        (else
          (bind _thermiteaudiocomponent (Variables|Default|GetThermiteAudioComponent)
            (:then
              (Audio|Components|Audio|FadeOut _thermiteaudiocomponent 0.5 0.0 "Linear" :!class /Script/Engine.AudioComponent))
            (:else)))))))

(event Custom|ListenForBossSetEvent
  (bind _vaulteventrouter (Variables|Default|GetVaultEventRouter)
    (:then
      (bind (_asynctaskproxy _payload) (GameplayEventRouter|StartListeningToStatefulEvent _vaulteventrouter "/VividRazorCoreGameplay/Gameplay/Structs/Struct_VividRazorCore_Vault_BossSet_Playload.Struct_VividRazorCore_Vault_BossSet_Playload")
        (:OnEventReceived
          (CallFunction|OnBossSetEvent (.ShowBossMinimapIcon_29_88FE3EE140BCC19BB36A5C864D18C7C0 (Utilities|Struct|BreakStructVividRazorCoreVaultBossSetPlayload _payload))))))
    (:else)))

(event Interaction|EventBlueprintOnInteract
  (params (InteractingPawn :type Fort%20Pawn%20Object%20Reference) (InteractionBeingAttempted :type EInteractionBeingAttempted%20Enum :default "FirstInteraction"))
  (if (CallFunction|IsPawnAllowedtoInteract InteractingPawn)
    (Networking|SwitchHasAuthority
      (:Authority
        (bind _retvalue (CallFunction|DoesPlayerHaveThermite InteractingPawn))
        (if _retvalue
          (if (not (Variables|/VividRazorVaultGameplay/Gameplay/Environmental/Manager/BGA_VividRazorVault_Manager.BGA_VividRazorVault_Manager_C|GetbAlert (Variables|Default|GetBGA_VaultManager)))
            (Class|BPIVividRazorCoreVaultEvents|RequestVaultStart(Message) self (Utilities|Struct|MakeStructVividRazorCoreVaultStartPayload (Variables|Default|GetVaultLocationTag) (Utilities|ResolveSoftReference InteractingPawn) (Utilities|ResolveSoftReference self)))
            (CallFunction|RemoveThermitefromPlayer (Fortnite|GetFortPlayerControllerfromActor InteractingPawn)))
          (else
            ; Failed to open vault GC
            (Ability|ExecuteGameplayCueReplicated (Variables|Default|GetGCFail))))))))

(event Custom|WeakpointUpdate
  (if (CallFunction|IsEnabledandActive)
    (Variables|Default|SetCurrentWeakpointLifetime (+ (Variables|Default|GetCurrentWeakpointLifetime) (Utilities|Time|GetTimerElapsedTimebyHandle (Variables|Default|GetTimerHandle_WeakpointUpdate))))
    (if (Variables|Default|GetbWeakpointIsExposed)
      (CallFunction|DamageWeakPoint (Variables|Default|GetCurrentWeakpoint) (CallFunction|GetWeakpointDamageOverTime)))))

(event EventHandleWeakPointUpdated
  (params (PointConfig :type Vault%20Weak%20Point%20Config%20Structure%20%28by%20ref%29) (NewHealthPercent :type Float%20%28single-precision%29))
  (bind (_shapename _socketname _hitns _locallocation _bisexposed) (Utilities|Struct|BreakVaultWeakPointConfig PointConfig))
  ; Weak Points
  (|CallParentFunction PointConfig NewHealthPercent)
  (Development|PrintString (Utilities|String|Append "weak point updated " (Utilities|String|ToString(Name) _shapename) " new health: " (Utilities|String|ToString(Float) NewHealthPercent)))
  (if _bisexposed
    (Visuals|SetTumblerProgress _shapename NewHealthPercent)))

(event EventHandleWeakPointSpawned
  (params (PointConfig :type Vault%20Weak%20Point%20Config%20Structure%20%28by%20ref%29))
  (bind (_shapename _socketname) (Utilities|Struct|BreakVaultWeakPointConfig PointConfig))
  (bind _returnvalue (CallFunction|GetWeakpointVulnerabilityDelay))
  (bind _rootmesh (Variables|VividRazorVaultDoor|GetRootMesh))
  (bind _ns_vividrazor_weakpoint_glow (Variables|Default|GetNS_VividRazor_Weakpoint_Glow))
  ; Weak Points
  (|CallParentFunction PointConfig)
  (Development|PrintString (Utilities|String|Append "Weak point spawned: " (Utilities|String|ToString(Name) _shapename)))
  (Visuals|SetTumblerHighlight _shapename)
  (CallFunction|OnTumblerSpawn _shapename)
  (Class|SoundEventListenerVividRazorVaultVaultDoor|WeakpointSpawn (Variables|Default|GetSoundEventListener_VividRazorVault_VaultDoor) _socketname)
  (Components|Activation|Activate _ns_vividrazor_weakpoint_glow false)
  (Transformation|AttachComponentToComponent _ns_vividrazor_weakpoint_glow _rootmesh _socketname "KeepRelative" "KeepRelative" "KeepRelative" true)
  (bind _returnvalue_2 (Niagara|SpawnSystemAttached "/VividRazorVaultGameplay/FX/NS_VividRazor_Weakpoint_Spawn.NS_VividRazor_Weakpoint_Spawn" _rootmesh (Utilities|String|StringToName (Utilities|String|Append (Utilities|String|ToString(Name) _socketname) "_Cover"))))
  (bind _output_get (Variables|Default|SetWeakpointParticleSystem _returnvalue_2))
  (Niagara|SetNiagaraVariable(Float) _output_get "LoopDuration" _returnvalue)
  (bind _returnvalue_3 (Utilities|Time|SetTimerbyEvent (EventDispatchers|CreateEvent :!event Custom|HideCovers) _returnvalue))
  (Variables|Default|SetCoverHideTimer _returnvalue_3))

(event Custom|HideCovers
  ; Weak Points
  (Visuals|SetCoverHide (Variables|Default|GetCurrentWeakpoint))
  (Variables|Default|SetbWeakpointIsExposed true)
  (CallFunction|SetWeakPointExposed (Variables|Default|GetCurrentWeakpoint) true)
  (Class|SoundEventListenerVividRazorVaultVaultDoor|WeakpointReady (Variables|Default|GetSoundEventListener_VividRazorVault_VaultDoor)))

(event EventHandleWeakPointDeath
  (params (PointConfig :type Vault%20Weak%20Point%20Config%20Structure%20%28by%20ref%29))
  (bind (_shapename _socketname _hitns _locallocation _bisexposed) (Utilities|Struct|BreakVaultWeakPointConfig PointConfig))
  ; Weak Points
  (|CallParentFunction PointConfig)
  (Development|PrintString (Utilities|String|Append "weak point died: " (Utilities|String|ToString(Name) _shapename)))
  (Utilities|FlowControl|Sequence
    (:then_0
      (Variables|Default|SetWeakpointAlive false)
      ;;[0.15 0.04398 0 0.5] Needed on both the server and the client so moving before branch
      (Class|BGAVividRazorVaultManager|BroadcastWeakpointDestroyedClient (Variables|Default|GetBGA_VaultManager))
      (if (Networking|IsServer)
        (Class|BGAVividRazorVaultManager|AddProgressWeakpoint (Variables|Default|GetBGA_VaultManager))
        (Utilities|Time|ClearandInvalidateTimerbyHandle (Variables|Default|GetTimerHandle_WeakpointUpdate))
        (Variables|Default|SetCurrentWeakpointLifetime 0.0)
        (Variables|Default|SetCurrentWeakpoint "None")
        (for _array_element (Variables|Default|GetContributingActorToDestroyWeakpoint)
          (Ability|SendGameplayEventtoActor _array_element "(TagName=\"Event.VividRazor.Vault.WeakpointDestroyed\")" (Utilities|Struct|MakeGameplayEventData "(TagName=\"Event.VividRazor.Vault.WeakpointDestroyed\")" self 0 0 0 "()" (Tags|GetGameplayTags))))
        (Utilities|Array|Clear (Variables|Default|GetContributingActorToDestroyWeakpoint))))
    (:then_1
      (Visuals|SetTumblerHide _shapename)
      (Variables|Default|SetbWeakpointIsExposed false)
      (Class|SoundEventListenerVividRazorVaultVaultDoor|WeakpointBreak (Variables|Default|GetSoundEventListener_VividRazorVault_VaultDoor)))
    (:then_2
      (if _bisexposed
        (CallFunction|SetWeakPointExposed (Variables|Default|GetCurrentWeakpoint))
        (else
          (Visuals|SetCoverHide (Variables|Default|GetCurrentWeakpoint))
          (Components|Activation|Deactivate (Variables|Default|GetWeakpointParticleSystem)))))))

(event EventHandleWeakPointDiedOutOfRelevancy
  (params (PointConfig :type Vault%20Weak%20Point%20Config%20Structure%20%28by%20ref%29))
  (bind _shapename (Utilities|Struct|BreakVaultWeakPointConfig PointConfig))
  ; Weak Points
  ;;[0.15 0.09344 0.000901 0.5] Sync up weak points destroyed before this actor was replicated to us
  (Visuals|SetCoverHide _shapename)
  (Visuals|SetTumblerHide _shapename))

(event Custom|NewWeakpoint
  ; Weak Points
  (Development|PrintString "Spawn new weak point")
  (CallFunction|SpawnNextWeakPoint)
  (Variables|Default|SetWeakpointAlive true)
  (Variables|Instigators|SetLastDamagerThisWeakpoint 0))

(event EventHandleWeakPointReceiveDamage
  (params (DamageInfo :type Vault%20Weak%20Point%20Damage%20Info%20Structure%20%28by%20ref%29))
  (bind (_damages _damagecauser) (Utilities|Struct|BreakVaultWeakPointDamageInfo DamageInfo))
  ; Weak Points
  (Networking|SwitchHasAuthority
    (:Remote
      (for _array_element _damages
        (Components|Activation|Activate (.HitNS (Utilities|Struct|BreakVaultWeakPointConfig (CallFunction|GetWeakPointConfig (Utilities|Struct|BreakVaultWeakPointRuntime _array_element)))) true)
        (Visuals|SetTumblerHit))
      (Class|BGAVividRazorVaultManager|BroadcastWeakpointDamagedClient (Variables|Default|GetBGA_VaultManager) (Utilities|ResolveSoftReference _damagecauser))
      (Class|SoundEventListenerVividRazorVaultVaultDoor|WeakpointDamaged (Variables|Default|GetSoundEventListener_VividRazorVault_VaultDoor)))))

(event EventOnDamageServer
  (params (Damage :type Float%20%28single-precision%29) (DamageTags :type Gameplay%20Tag%20Container%20Structure%20%28by%20ref%29) (Momentum :type Vector :default "0, 0, 0") (HitInfo :type Hit%20Result%20Structure%20%28by%20ref%29) (InstigatedBy :type Controller%20Object%20Reference) (DamageCauser :type Actor%20Object%20Reference) (EffectContext :type Gameplay%20Effect%20Context%20Handle%20Structure))
  ; Weak Points
  (if (> Damage 0)
    (Utilities|Array|AddUnique (Variables|Default|GetContributingActorToOpenVault) InstigatedBy)
    (Utilities|Array|AddUnique (Variables|Default|GetContributingActorToDestroyWeakpoint) InstigatedBy)
    (Analytics|LogWeakpointDamageData Damage DamageTags)))

(event EventResetVault
  (|CallParentFunction)
  (CallFunction|ResetReplicatedProperties)
  (CallFunction|ResetVisuals)
  (Class|BGAVividRazorVaultManager|ResetVault (Variables|Default|GetBGA_VaultManager))
  (Variables|Instigators|SetVaultStarter 0)
  (Variables|Instigators|SetLastDamager 0)
  (Variables|Instigators|SetLastDamagerThisWeakpoint 0))

(event Custom|ResetReplicatedProperties
  (Variables|Interaction|SetbAllowInteract true)
  (Variables|Default|SetWeakpointAlive false))

(event Custom|ResetVisuals
  (bind _rootmesh (Variables|VividRazorVaultDoor|GetRootMesh))
  (bind _rootmesh_2 (Variables|VividRazorVaultDoor|GetRootMesh))
  (bind _rootmesh_3 (Variables|VividRazorVaultDoor|GetRootMesh))
  (Utilities|FlowControl|Sequence
    (:then_0
      (for _array_element (CallFunction|GetAllWeakPointConfigs)
        (for _array_element_2 (Variables|Default|GetCoverMaterialIndex)
          (Rendering|Material|SetScalarParameterValue (Class|BTechArtGlobalLibrary|GetDynamicMaterialInstance _rootmesh_2 _array_element_2) (Utilities|String|StringToName (Utilities|String|Append "Hide" (Utilities|String|ToString(Name) (Utilities|Struct|BreakVaultWeakPointConfig _array_element)))) 0.0 :!class /Script/Engine.MaterialInstanceDynamic)
          (Rendering|Material|SetScalarParameterValue (Class|BTechArtGlobalLibrary|GetDynamicMaterialInstance _rootmesh_2 _array_element_2) (Utilities|String|StringToName (Utilities|String|Append "Highlight" (Utilities|String|ToString(Name) (Utilities|Struct|BreakVaultWeakPointConfig _array_element)))) 0.0 :!class /Script/Engine.MaterialInstanceDynamic))
        (for _array_element_3 (Variables|Default|GetTumblerMaterialIndex)
          (Rendering|Material|SetScalarParameterValue (Class|BTechArtGlobalLibrary|GetDynamicMaterialInstance _rootmesh _array_element_3) (Utilities|String|StringToName (Utilities|String|Append "Hide" (Utilities|String|ToString(Name) (Utilities|Struct|BreakVaultWeakPointConfig _array_element)))) 0.0 :!class /Script/Engine.MaterialInstanceDynamic)
          (Rendering|Material|SetScalarParameterValue (Class|BTechArtGlobalLibrary|GetDynamicMaterialInstance _rootmesh _array_element_3) (Utilities|String|StringToName (Utilities|String|Append "Damage" (Utilities|String|ToString(Name) (Utilities|Struct|BreakVaultWeakPointConfig _array_element)))) 0.0 :!class /Script/Engine.MaterialInstanceDynamic)
          (Rendering|Material|SetScalarParameterValue (Class|BTechArtGlobalLibrary|GetDynamicMaterialInstance _rootmesh _array_element_3) (Utilities|String|StringToName (Utilities|String|Append "Highlight" (Utilities|String|ToString(Name) (Utilities|Struct|BreakVaultWeakPointConfig _array_element)))) 0.0 :!class /Script/Engine.MaterialInstanceDynamic)
          (Niagara|SetNiagaraVariable(Float) (Variables|Default|GetNS_VividRazor_Weakpoint_Glow) "Damage" 0.0)
          (Class|SoundEventListenerVividRazorVaultVaultDoor|OnWeakpointHealthChanged (Variables|Default|GetSoundEventListener_VividRazorVault_VaultDoor) 1.0))))
    (:then_1
      (Components|StaticMesh|SetStaticMesh (Variables|VividRazorVaultDoor|GetRootMesh) "/VividRazorVaultGameplay/Assets/Gameplay_VaultDoor_C/Meshes/SM_Gameplay_VaultDoor_C_B.SM_Gameplay_VaultDoor_C_B")
      (Rendering|Material|SetMaterial _rootmesh_3 0 "/VividRazorVaultGameplay/Assets/Gameplay_VaultDoor_C/Materials/MI_Gameplay_VaultDoor_C_A.MI_Gameplay_VaultDoor_C_A")
      (Rendering|Material|SetMaterial _rootmesh_3 1 "/VividRazorVaultGameplay/Assets/Gameplay_VaultDoor_C/Materials/MI_Gameplay_VaultDoor_Lights_C_A.MI_Gameplay_VaultDoor_Lights_C_A")
      (Visuals|SetOverallHeatingVisuals 0.0)
      (Rendering|SetVisibility (Variables|Default|GetThermite) true true)
      (Visuals|UpdateThermiteMaterials false)
      (Components|Activation|Deactivate (Variables|Default|GetNS_Thermite_for_Vault_Proto))
      (Utilities|Time|ClearandInvalidateTimerbyHandle (Variables|Default|GetCoverHideTimer))
      (Visuals|UpdateLights false)
      (CallFunction|ThermiteAudio false))))

(event Custom|Handle%20Boss%20Spawned
  (Variables|Default|SetbIsBossPresent true))

(event Custom|Handle%20Boss%20Defeated
  (Variables|Default|SetbIsBossPresent false))

(event Custom|Handle%20Encounter%20Completed
  (bind _npcmusiccomponent_vividrazorvaults (Variables|Default|GetNPCMusicComponent_VividRazorVaults)
    (:then
      (Class|NPCMusicComponentVividRazorVaults|TransitionToNextMusicStage _npcmusiccomponent_vividrazorvaults "NewEnumerator3"))
    (:else)))

(event EventSetLastDamageInstigator
  (params (DamageInstigator :type Controller%20Object%20Reference))
  (Utilities|IsValid DamageInstigator
    (:Is%20Valid
      (Variables|Instigators|SetLastDamager DamageInstigator)
      (Variables|Instigators|SetLastDamagerThisWeakpoint DamageInstigator))))

(event Custom|HandleAlertPaused
  (Utilities|FlowControl|Sequence
    (:then_0
      (CallFunction|DestroyActiveWeakpoint))
    (:then_1
      (Visuals|UpdateThermiteMaterials false)
      (Variables|Default|SetbIsActive false)
      (Components|Activation|Deactivate (Variables|Default|GetNS_Thermite_for_Vault_Proto))
      (Variables|Interaction|SetbAllowInteract (not (Variables|VividRazorVaultDoor|GetbDoorOpened)))
      (Variables|Default|SetWeakpointAlive false)
      (Class|SoundEventListenerVividRazorVault|OnVaultCracked (Variables|Default|GetSoundEventListener_VividRazorVault_VaultDoor)) ; Stops vault alarms
      (CallFunction|ThermiteAudio false)
      (Variables|Default|SetMusic%20Stage "NewEnumerator0"))))

(fn BlueprintCanInteract
  (params (InteractingPawn :type Fort%20Pawn%20Object%20Reference) (InteractionBeingAttempted :type EInteractionBeingAttempted%20Enum :default "FirstInteraction") (InteractionType :type TInteractionType%20Enum :default "IT_NoInteraction"))
  (if (CallFunction|IsPawnAllowedtoInteract InteractingPawn)
    (Variables|Default|GetBGA_VaultManager
      (:then
        (if (Variables|/VividRazorVaultGameplay/Gameplay/Environmental/Manager/BGA_VividRazorVault_Manager.BGA_VividRazorVault_Manager_C|GetbAlert (Variables|Default|GetBGA_VaultManager))
          (return false)))
      (:else
        _))
    (Utilities|IsValid InteractingPawn
      (:Is%20Valid
        (bind _retvalue (CallFunction|DoesPlayerHaveThermite InteractingPawn))
        (return _retvalue)))
    (else
      (return false))))

(fn BlueprintGetFailedInteractionString
  (params (InteractingPawn :type Fort%20Pawn%20Object%20Reference) (InteractionBeingAttempted :type EInteractionBeingAttempted%20Enum :default "FirstInteraction"))
  (if (CallFunction|IsPawnAllowedtoInteract InteractingPawn)
    (Variables|Default|GetBGA_VaultManager
      (:then
        (if (Variables|/VividRazorVaultGameplay/Gameplay/Environmental/Manager/BGA_VividRazorVault_Manager.BGA_VividRazorVault_Manager_C|GetbAlert (Variables|Default|GetBGA_VaultManager))
          (return "")
          (else
            _)))
      (:else
        _))
    (return "NSLOCTEXT(\"[2CD68F5F38C39BB76F7DCEF5FF667DF0]\", \"FD72B3A0496CB2D67BF5ACACE2F23112\", \"Requires Thermite\")")
    (else
      (return ""))))

(fn BlueprintGetInteractionString
  (params (InteractingPawn :type Fort%20Pawn%20Object%20Reference) (InteractionBeingAttempted :type EInteractionBeingAttempted%20Enum :default "FirstInteraction"))
  (Utilities|FlowControl|Sequence
    (:then_0
      (Utilities|IsValid InteractingPawn
        (:Is%20Valid
          (bind _retvalue (CallFunction|DoesPlayerHaveThermite InteractingPawn))
          (if _retvalue
            (return "NSLOCTEXT(\"[2CD68F5F38C39BB76F7DCEF5FF667DF0]\", \"EF7536474D9526C78E8094B3ABC6ED1C\", \"Place Thermite\")")))))
    (:then_1
      (return "NSLOCTEXT(\"[2CD68F5F38C39BB76F7DCEF5FF667DF0]\", \"87827EC4422B733AFC6528A5CB3407E3\", \"Requires Thermite\")"))))

(fn BlueprintGetInteractionTime
  (params (InteractingPawn :type Fort%20Pawn%20Object%20Reference) (InteractionBeingAttempted :type EInteractionBeingAttempted%20Enum :default "FirstInteraction"))
  (return 0.5 true))

(fn BP_ModifyIncomingDamage
  (params (Damage :type Float%20%28single-precision%29) (InTags :type Gameplay%20Tag%20Container%20Structure%20%28by%20ref%29) (EffectContext :type Gameplay%20Effect%20Context%20Handle%20Structure%20%28by%20ref%29) (EventInstigator :type Controller%20Object%20Reference) (DamageCauser :type Actor%20Object%20Reference))
  (if (CallFunction|IsEnabledandActive)
    (if (Variables|Default|GetbWeakpointIsExposed)
      (return (Default|CheckDamageMultipliers (Variables|GetDamage) InTags))
      (else
        (return 0.0)))
    (else
      (return 0.0))))

(fn CheckDamageMultipliers
  (params (Damage :type Float%20%28double-precision%29) (Tags :type Gameplay%20Tag%20Container%20Structure))
  ;[0.03429 0.3069 0.7899] Iterate our damage multiplier data and see if this instance of damage should be multiplied
  (for _array_element (Variables|ScalableFloats|GetDamageMultiplers)
    (bind (_array_element_tags_3_9392e1e541244de4e3cfecbc304111d3 _array_element_multiplier_6_dec6c6724a29df178fceabaa6e8a8e06) (Utilities|Struct|BreakStructVividRazorVaultDamageMultiplier _array_element))
    (if (GameplayTags|HasAnyTags (Variables|Actor|GetTags) _array_element_tags_3_9392e1e541244de4e3cfecbc304111d3 true)
      (return (* (Variables|GetDamage) (ScalableFloat|GetValueatLevel _array_element_multiplier_6_dec6c6724a29df178fceabaa6e8a8e06)))))
  (return Damage))

(fn ConnectToManager
  (params (ManagerActor :type BGA%20Vivid%20Razor%20Vault%20Manager%20Object%20Reference))
  (bind _output_get (Variables|Default|SetBGA_VaultManager ManagerActor))
  (CallFunction|OnBossSetEvent (Variables|/VividRazorVaultGameplay/Gameplay/Environmental/Manager/BGA_VividRazorVault_Manager.BGA_VividRazorVault_Manager_C|GetShowBossOnMinimap _output_get)))

(fn DestroyActiveWeakpoint
  (CallFunction|DamageWeakPoint (Variables|Default|GetCurrentWeakpoint) 5e+04))

(fn DoesPlayerHaveThermite
  (params (PlayerPawn :type Fort%20Pawn%20Object%20Reference))
  (bind _returnvalue (Fortnite|GetFortPlayerControllerfromActor PlayerPawn))
  (Utilities|FlowControl|Sequence
    (:then_0
      (Utilities|IsValid _returnvalue
        (:Is%20Valid
          (Utilities|IsValid (Inventory|FindExistingItemForDefinition :self _returnvalue :ItemDefinition (Variables|Default|GetThermiteItem) :bInStorageVault false)
            (:Is%20Valid
              (return true))
            (:Is%20Not%20Valid
              (Utilities|IsValid (Inventory|FindExistingItemForDefinition :self _returnvalue :ItemDefinition (Variables|Default|GetThermiteItem_Keycard) :bInStorageVault false)
                (:Is%20Valid
                  (return true))))))))
    (:then_1
      (return false))))

(fn GetMusicStage
  (return (Variables|Default|GetMusic%20Stage)))

(fn GetPercentDamageForWeakpointStage
  (return (select (Class|BGAVividRazorVaultManager|IsDrillAttached (Variables|Default|GetBGA_VaultManager)) (select (>= (Variables|Default|GetCurrentWeakpointLifetime) (ScalableFloat|GetValueatLevel (Variables|ScalableFloats|GetROW_Weakpoint_Stage3_StartTime))) (ScalableFloat|GetValueatLevel (Variables|ScalableFloats|GetROW_Weakpoint_DamagePercent_Stage3_WithDrill)) (select (>= (Variables|Default|GetCurrentWeakpointLifetime) (ScalableFloat|GetValueatLevel (Variables|ScalableFloats|GetROW_Weakpoint_Stage2_StartTime))) (ScalableFloat|GetValueatLevel (Variables|ScalableFloats|GetROW_Weakpoint_DamagePercent_Stage2_WithDrill)) (ScalableFloat|GetValueatLevel (Variables|ScalableFloats|GetROW_Weakpoint_DamagePercent_Stage1_WithDrill)))) (select (>= (Variables|Default|GetCurrentWeakpointLifetime) (ScalableFloat|GetValueatLevel (Variables|ScalableFloats|GetROW_Weakpoint_Stage3_StartTime))) (ScalableFloat|GetValueatLevel (Variables|ScalableFloats|GetROW_Weakpoint_DamagePercent_Stage3)) (select (>= (Variables|Default|GetCurrentWeakpointLifetime) (ScalableFloat|GetValueatLevel (Variables|ScalableFloats|GetROW_Weakpoint_Stage2_StartTime))) (ScalableFloat|GetValueatLevel (Variables|ScalableFloats|GetROW_Weakpoint_DamagePercent_Stage2)) (ScalableFloat|GetValueatLevel (Variables|ScalableFloats|GetROW_Weakpoint_DamagePercent_Stage1)))))))

(fn GetTumblerID
  (params (Selection :type Name)))

(fn GetWeakpointDamageOverTime
  (return (* (Math|Float|SafeDivide (ScalableFloat|GetValueatLevel (Variables|ScalableFloats|GetROW_WeakpointMaxHealth)) (ScalableFloat|GetValueatLevel (Variables|ScalableFloats|GetROW_WeakpointMaxLifetime))) (CallFunction|GetPercentDamageforWeakpointStage))))

(fn GetWeakpointVulnerabilityDelay
  (return (Math|Float|SelectFloat (ScalableFloat|GetValueatLevel (Variables|ScalableFloats|GetROW_Weakpoint_CoverRemovalTime_FirstWeakpoint)) (ScalableFloat|GetValueatLevel (Variables|ScalableFloats|GetROW_Weakpoint_CoverRemovalTime)) (== (Variables|/VividRazorVaultGameplay/Gameplay/Environmental/Manager/BGA_VividRazorVault_Manager.BGA_VividRazorVault_Manager_C|GetNumWeakpointsSpawned (Variables|Default|GetBGA_VaultManager)) 0))))

(fn InitializeMusic
  (if (Variables|Default|GetbIsBossPresent)
    (bind _npcmusiccomponent_vividrazorvaults (Variables|Default|GetNPCMusicComponent_VividRazorVaults)
      (:then
        (Class|NPCMusicComponentVividRazorVaults|BossSpawned _npcmusiccomponent_vividrazorvaults))
      (:else))
    (else
      _))
  (Default|SetMusicStage (Variables|Default|GetNPCMusicComponent_VividRazorVaults) (Variables|Default|GetMusic%20Stage)))

(fn InitializeDynamicMaterial
  (bind _rootmesh (Variables|VividRazorVaultDoor|GetRootMesh))
  (bind _thermite (Variables|Default|GetThermite))
  (for (_array_element _array_index) (Rendering|Material|GetMaterials _rootmesh)
    (Utilities|FlowControl|Sequence
      (:then_0
        (if (Utilities|String|Contains (Utilities|String|ToLower (Utilities|GetDisplayName _array_element)) "tumblers")
          (Utilities|Array|Add (Variables|Default|GetTumblerMaterialIndex) _array_index)
          (elif (Utilities|String|Contains (Utilities|String|ToLower (Utilities|GetDisplayName _array_element)) "cover")
            (Utilities|Array|Add (Variables|Default|GetCoverMaterialIndex) _array_index)
            (elif (not (Utilities|String|Contains (Utilities|String|ToLower (Utilities|GetDisplayName _array_element)) "light"))
              (Utilities|Array|Add (Variables|Default|GetDoorMaterialIndex) _array_index)
              (else
                (Variables|Default|SetVaultLightsMaterial (Class|BTechArtGlobalLibrary|GetDynamicMaterialInstance _rootmesh _array_index)))))))))
  (Utilities|FlowControl|Sequence
    (:then_0
      (Variables|Visuals|SetMI_Thermite_0 (Rendering|Material|GetMaterial _thermite 0))
      (Variables|Visuals|SetMI_Thermite_1 (Rendering|Material|GetMaterial _thermite 1))
      (Variables|Visuals|SetMI_Thermite_2 (Rendering|Material|GetMaterial _thermite 2))
      (Variables|Visuals|SetMI_Thermite_3 (Rendering|Material|GetMaterial _thermite 3))
      (Variables|Visuals|SetMI_Thermite_4 (Rendering|Material|GetMaterial _thermite 4))
      (Variables|Visuals|SetMI_Thermite_5 (Rendering|Material|GetMaterial _thermite 5)))
    (:then_1
      (Visuals|UpdateThermiteMaterials false))))

(fn IsEnabledAndActive
  (return (and (Variables|Default|GetbIsEnabled) (Variables|Default|GetbIsActive))))

(fn IsPawnAllowedToInteract
  (params (InteractingPawn :type Fort%20Pawn%20Object%20Reference))
  (return (Variables|Default|GetbIsEnabled)))

(fn IsTrainVault
  (return false))

(fn IsVaultAlertPayloadMine
  (params (Struct_VividRazorCore_Vault_Alert_Payload :type Struct%20Vivid%20Razor%20Core%20Vault%20Alert%20Payload%20Structure%20%28by%20ref%29))
  (return (== (Utilities|Struct|BreakStructVividRazorCoreVaultAlertPayload Struct_VividRazorCore_Vault_Alert_Payload) (Variables|Default|GetVaultLocationTag))))

(fn IsVaultOpenedPayloadMine
  (params (Struct_VividRazorCore_Vault_Opened_Payload :type Struct%20Vivid%20Razor%20Core%20Vault%20Opened%20Payload%20Structure%20%28by%20ref%29))
  (return (== (Utilities|Struct|BreakStructVividRazorCoreVaultOpenedPayload Struct_VividRazorCore_Vault_Opened_Payload) (Variables|Default|GetVaultLocationTag))))

(fn IsVaultStarted
  (bind _bga_vaultmanager (Variables|Default|GetBGA_VaultManager)
    (:then
      (bind _isalert (Class|BGAVividRazorVaultManager|GetBAlert _bga_vaultmanager))
      (return _isalert))
    (:else
      (return false))))

(fn IsWeakpointAlive
  (return (Variables|Default|GetWeakpointAlive)))

(fn LocalOnFailedInteract
  (params (InteractingPawn :type Fort%20Player%20Pawn%20Object%20Reference)))

(fn LogWeakpointDamageData
  (params (PlayerWeakpointDamage :type Float%20%28double-precision%29) (DamageTags :type Gameplay%20Tag%20Container%20Structure))
  (Networking|SwitchHasAuthority
    (:Authority
      (LocalVariables|SetDamageType "Other")
      (for _array_element (Variables|Analytics|GetDamageTagsToLog)
        (if (GameplayTags|HasAnyTags (Variables|GetDamageTags) _array_element true)
          (LocalVariables|SetDamageType (GameplayTags|GetDebugStringfromGameplayTagContainer _array_element))))
      (Analytics|SavePlayerWeakpointDamage (Variables|Default|GetBGA_VaultManager) (Variables|GetPlayerWeakpointDamage) (LocalVariables|GetDamageType)))))

(fn MakeGameplayCueDoorOpenParams
  (return (Ability|GameplayCue|MakeGameplayCueParameters :NormalizedMagnitude 0.0 :RawMagnitude 0.0 :Location (Transformation|GetActorLocation) :Normal "0,0,0" :GameplayEffectLevel 1 :AbilityLevel 1 :bReplicateLocationWhenUsingMinimalRepProxy false)))

(fn OnBossSetEvent
  (params (ShowBossOnMinimap :type Boolean)))

(fn OnDoorOpen
  (CallFunction|UpdateDoorVisualsonOpen)
  (Utilities|FlowControl|Sequence
    (:then_0
      (for _array_element (Actor|GetAttachedActors true)
        (Class|BPIVividRazorCoreVaultEvents|VaultOpened(Message) _array_element))
      (for _array_element_2 (Variables|Default|GetContributingActorToOpenVault)
        (Ability|SendGameplayEventtoActor _array_element_2 "(TagName=\"Event.VividRazor.Vault.Opened\")" (Utilities|Struct|MakeGameplayEventData "(TagName=\"Event.VividRazor.Vault.Opened\")" self 0 0 0 "()" (Tags|GetGameplayTags)))))
    (:then_1
      (if (Networking|IsServer)
        (Ability|ExecuteGameplayCueWithParamsReplicated (Variables|Default|GetGC_DoorOpened) (Visuals|MakeGameplayCueDoorOpenParams))))
    (:then_2
      ; Prevent mobile EasyInteract button from appearing, especially when AlertProgress is reset
      (Variables|Interaction|SetbAllowInteract false))))

(fn OnRep_bIsBossPresent
  (if (Variables|Default|GetbIsBossPresent)
    (bind _npcmusiccomponent_vividrazorvaults (Variables|Default|GetNPCMusicComponent_VividRazorVaults)
      (:then
        (Class|NPCMusicComponentVividRazorVaults|BossSpawned _npcmusiccomponent_vividrazorvaults))
      (:else))
    (else
      (bind _npcmusiccomponent_vividrazorvaults_2 (Variables|Default|GetNPCMusicComponent_VividRazorVaults)
        (:then
          (Class|NPCMusicComponentVividRazorVaults|BossDefeated _npcmusiccomponent_vividrazorvaults_2))
        (:else)))))

(fn OnRep_MusicStage
  (bind _npcmusiccomponent_vividrazorvaults (Variables|Default|GetNPCMusicComponent_VividRazorVaults)
    (:then
      (Class|NPCMusicComponentVividRazorVaults|TransitionToNextMusicStage _npcmusiccomponent_vividrazorvaults (Variables|Default|GetMusic%20Stage)))
    (:else)))

(fn OnTumblerSpawn
  (params (WeakpointShapeName :type Name))
  (Variables|Default|SetCurrentWeakpoint WeakpointShapeName)
  (Variables|Default|SetCurrentWeakpointLifetime 0.0)
  (Networking|SwitchHasAuthority
    (:Authority
      (Utilities|Time|ClearandInvalidateTimerbyHandle (Variables|Default|GetTimerHandle_WeakpointUpdate))
      (bind _returnvalue (Utilities|Time|SetTimerbyEvent (EventDispatchers|CreateEvent :!event WeakpointUpdate) 1.0 true true))
      (Variables|Default|SetTimerHandle_WeakpointUpdate _returnvalue))))

(fn RemoveThermiteFromPlayer
  (params (Player%20Controller :type Fort%20Player%20Controller%20Object%20Reference))
  ; Remove Thermite from inventory
  (Inventory|RemoveItemFromPlayer :PlayerController Player%20Controller :ItemDefinition (Variables|Default|GetThermiteItem) :AmountToRemove 1 :bForceRemoval true :bSkipVerbRemoval false)
  ; Support keycard version for now as well
  (Inventory|RemoveItemFromPlayer :PlayerController Player%20Controller :ItemDefinition (Variables|Default|GetThermiteItem_Keycard) :AmountToRemove 1 :bForceRemoval true :bSkipVerbRemoval false))

(fn RequestTrainVaultStart
  (params (Payload :type Struct%20Vivid%20Razor%20Core%20Vault%20Start%20Payload%20Structure))
  (return false))

(fn RequestVaultStart
  (params (Payload :type Struct%20Vivid%20Razor%20Core%20Vault%20Start%20Payload%20Structure))
  (bind (_vaultlocationtag_24_4638c5e048c21e5b7f1d909063318292 _triggeringpawn_33_567a525a4447d3f140bec19d19506604) (Utilities|Struct|BreakStructVividRazorCoreVaultStartPayload Payload))
  (bind (_payload_vaultlocationtag_24_4638c5e048c21e5b7f1d909063318292 _payload_triggeringpawn_33_567a525a4447d3f140bec19d19506604) (Utilities|Struct|BreakStructVividRazorCoreVaultStartPayload (Variables|GetPayload_VaultLocationTag_24_4638C5E048C21E5B7F1D909063318292)))
  (bind _output (Utilities|ResolveSoftReference _payload_triggeringpawn_33_567a525a4447d3f140bec19d19506604))
  (bind _returnvalue (Pawn|GetController _output))
  (if (CallFunction|IsPawnAllowedtoInteract (Utilities|ResolveSoftReference _triggeringpawn_33_567a525a4447d3f140bec19d19506604))
    (bind _structout (Utilities|Struct|SetmembersinStructVividRazorCoreVaultStartPayload Payload (Variables|Default|GetVaultLocationTag)))
    (Class|BGAVividRazorVaultManager|StartVault (Variables|Default|GetBGA_VaultManager) _structout)
    (Default|CallVaultGameplayStarted)
    (Ability|SendGameplayEventtoActor _output "(TagName=\"Event.VividRazor.Thermite.Attached.Vault\")" (Utilities|Struct|MakeGameplayEventData "(TagName=\"Event.VividRazor.Thermite.Attached.Vault\")" self))
    (Utilities|Array|AddUnique (Variables|Default|GetContributingActorToOpenVault) _returnvalue)
    (Variables|Analytics|SetLastVaultStarter _returnvalue)
    (return false)))

(fn SetCoverHide
  (params (Selection :type Name))
  (for _array_element (Variables|Default|GetCoverMaterialIndex)
    (Rendering|Material|SetScalarParameterValue (Class|BTechArtGlobalLibrary|GetDynamicMaterialInstance (Variables|VividRazorVaultDoor|GetRootMesh) _array_element) (Utilities|String|StringToName (Utilities|String|Append "Hide" (Utilities|String|ToString(Name) Selection))) 1.0 :!class /Script/Engine.MaterialInstanceDynamic)
    (Niagara|SetNiagaraVariable(Float) (Variables|Default|GetNS_VividRazor_Weakpoint_Glow) "Damage" 0.0)
    (Niagara|SpawnSystemAttached "/VividRazorVaultGameplay/FX/NS_VividRazor_Vault_CoverBreak.NS_VividRazor_Vault_CoverBreak" (Variables|VividRazorVaultDoor|GetRootMesh) Selection)
    (Class|SoundEventListenerVividRazorVaultVaultDoor|OnWeakpointHealthChanged (Variables|Default|GetSoundEventListener_VividRazorVault_VaultDoor) 1.0)))

(fn SetOverallHeatingVisuals
  (params (Value :type Float%20%28single-precision%29))
  (for _array_element (Variables|Default|GetDoorMaterialIndex)
    (Rendering|Material|SetScalarParameterValue (Class|BTechArtGlobalLibrary|GetDynamicMaterialInstance (Variables|VividRazorVaultDoor|GetRootMesh) _array_element) "OverallOverheatAmount" Value :!class /Script/Engine.MaterialInstanceDynamic)))

(fn SetTumblerHide
  (params (Selection :type Name))
  (for _array_element (Variables|Default|GetTumblerMaterialIndex)
    (Rendering|Material|SetScalarParameterValue (Class|BTechArtGlobalLibrary|GetDynamicMaterialInstance (Variables|VividRazorVaultDoor|GetRootMesh) _array_element) (Utilities|String|StringToName (Utilities|String|Append "Hide" (Utilities|String|ToString(Name) Selection))) 1.0 :!class /Script/Engine.MaterialInstanceDynamic)
    (Niagara|SetNiagaraVariable(Float) (Variables|Default|GetNS_VividRazor_Weakpoint_Glow) "Damage" 0.0)
    (Class|SoundEventListenerVividRazorVaultVaultDoor|OnWeakpointHealthChanged (Variables|Default|GetSoundEventListener_VividRazorVault_VaultDoor) 1.0)))

(fn SetTumblerHighlight
  (params (Selection :type Name) (ResetTo0 :type Boolean))
  (bind _returnvalue (Utilities|String|ToString(Name) Selection))
  (bind _returnvalue_2 (Utilities|String|Append "Highlight" _returnvalue))
  (bind _returnvalue_3 (Utilities|String|StringToName _returnvalue_2))
  (bind _rootmesh (Variables|VividRazorVaultDoor|GetRootMesh))
  (Utilities|FlowControl|Sequence
    (:then_0
      (for _array_element (Variables|Default|GetTumblerMaterialIndex)
        (Rendering|Material|SetScalarParameterValue (Class|BTechArtGlobalLibrary|GetDynamicMaterialInstance (Variables|VividRazorVaultDoor|GetRootMesh) _array_element) _returnvalue_3 (select (Variables|GetResetTo0) 0.0 1.0) :!class /Script/Engine.MaterialInstanceDynamic)))
    (:then_1
      (for _array_element_2 (Variables|Default|GetCoverMaterialIndex)
        (Rendering|Material|SetScalarParameterValue (Class|BTechArtGlobalLibrary|GetDynamicMaterialInstance _rootmesh _array_element_2) _returnvalue_3 (select (Variables|GetResetTo0) 0.0 1.0) :!class /Script/Engine.MaterialInstanceDynamic)
        (Rendering|Material|SetScalarParameterValue (Class|BTechArtGlobalLibrary|GetDynamicMaterialInstance _rootmesh _array_element_2) "OverheatTimestamp" (select (Variables|GetResetTo0) 0.0 (Utilities|Time|GetGameTimeinSeconds)) :!class /Script/Engine.MaterialInstanceDynamic)
        (Rendering|Material|SetScalarParameterValue (Class|BTechArtGlobalLibrary|GetDynamicMaterialInstance _rootmesh _array_element_2) "OverheatTimestampDuration" (CallFunction|GetWeakpointVulnerabilityDelay) :!class /Script/Engine.MaterialInstanceDynamic)))))

(fn SetTumblerHit
  (for _array_element (Variables|Default|GetTumblerMaterialIndex)
    (Rendering|Material|SetScalarParameterValue (Class|BTechArtGlobalLibrary|GetDynamicMaterialInstance (Variables|VividRazorVaultDoor|GetRootMesh) _array_element) "HitTimestamp" (Utilities|Time|GetGameTimeinSeconds) :!class /Script/Engine.MaterialInstanceDynamic)))

(fn SetTumblerProgress
  (params (Selection :type Name) (Damage :type Float%20%28double-precision%29))
  (bind _returnvalue (- 1.0 Damage))
  (for _array_element (Variables|Default|GetTumblerMaterialIndex)
    (Rendering|Material|SetScalarParameterValue (Class|BTechArtGlobalLibrary|GetDynamicMaterialInstance (Variables|VividRazorVaultDoor|GetRootMesh) _array_element) (Utilities|String|StringToName (Utilities|String|Append "Damage" (Utilities|String|ToString(Name) Selection))) _returnvalue :!class /Script/Engine.MaterialInstanceDynamic)
    (Niagara|SetNiagaraVariable(Float) (Variables|Default|GetNS_VividRazor_Weakpoint_Glow) "Damage" _returnvalue)
    (Class|SoundEventListenerVividRazorVaultVaultDoor|OnWeakpointHealthChanged (Variables|Default|GetSoundEventListener_VividRazorVault_VaultDoor) (Variables|GetDamage))))

(fn ShouldDie
  (params (Damage :type Float%20%28single-precision%29) (EventInstigator :type Controller%20Object%20Reference) (DamageCauser :type Actor%20Object%20Reference) (bIsTest :type Boolean) (EffectContext :type Gameplay%20Effect%20Context%20Handle%20Structure%20%28by%20ref%29))
  (return false))

(fn UpdateDoorVisualsOnOpen
  (Components|StaticMesh|SetStaticMesh (Variables|VividRazorVaultDoor|GetRootMesh) "/VividRazorVaultGameplay/Assets/Gameplay_VaultDoor_B/Meshes/SM_Gameplay_VaultDoor_B_B.SM_Gameplay_VaultDoor_B_B")
  (Rendering|Material|SetMaterial (Variables|VividRazorVaultDoor|GetRootMesh) 0 "/VividRazorVaultGameplay/Assets/Gameplay_VaultDoor_D/Materials/MI_Gameplay_VaultDoor_D_A.MI_Gameplay_VaultDoor_D_A")
  (Visuals|SetOverallHeatingVisuals 0.0)
  (Rendering|SetVisibility (Variables|Default|GetThermite) false false)
  (Components|Activation|Deactivate (Variables|Default|GetNS_Thermite_for_Vault_Proto))
  (Utilities|Time|ClearandInvalidateTimerbyHandle (Variables|Default|GetCoverHideTimer))
  (Visuals|UpdateLights false))

(fn UpdateLights
  (params (On :type Boolean))
  (bind _vaultlightsmaterial (Variables|Default|GetVaultLightsMaterial)
    (:then
      (Rendering|Material|SetVectorParameterValue _vaultlightsmaterial "SFX_Emissive_Tint(Expand)" (Math|Conversions|ToLinearColor(Float) (Math|Float|SelectFloat 1.0 0.1 On)) :!class /Script/Engine.MaterialInstanceDynamic)
      (Rendering|Material|SetScalarParameterValue _vaultlightsmaterial "Emissive Time Speed Layer 1" (Math|Float|SelectFloat 0.5 0.0 On) :!class /Script/Engine.MaterialInstanceDynamic))
    (:else)))

(fn UpdateThermiteMaterials
  (params (ShowRealMesh :type Boolean))
  (bind _thermite (Variables|Default|GetThermite))
  (if ShowRealMesh
    (Rendering|SetVisibility (Variables|Default|GetThermite) true true)
    (Rendering|Material|SetMaterial _thermite 0 (Variables|Visuals|GetMI_Thermite_0))
    (Rendering|Material|SetMaterial _thermite 1 (Variables|Visuals|GetMI_Thermite_1))
    (Rendering|Material|SetMaterial _thermite 2 (Variables|Visuals|GetMI_Thermite_2))
    (Rendering|Material|SetMaterial _thermite 3 (Variables|Visuals|GetMI_Thermite_3))
    (Rendering|Material|SetMaterial _thermite 4 (Variables|Visuals|GetMI_Thermite_4))
    (Rendering|Material|SetMaterial _thermite 5 (Variables|Visuals|GetMI_Thermite_5))
    (elif (ScalableFloat|GetValueasBool (Variables|ScalableFloats|GetROW_Thermite_Hologram_Enabled))
      (Rendering|SetVisibility (Variables|Default|GetThermite) true true)
      (Rendering|Material|SetMaterial (Variables|Default|GetThermite) 0 (Variables|Visuals|GetMI_Thermite_Hologram))
      (else
        ; FORT-844553: Hologram isn't enabled (primarily for capture team, hide the thermite on the vault
        (Rendering|SetVisibility (Variables|Default|GetThermite) false true)))))

(fn ConstructionScript
  (|CallParentFunction)
  (Variables|/VividRazorVaultGameplay/Assets/Audio/Blueprints/SoundEventListener_VividRazorVault.SoundEventListener_VividRazorVault_C|SetVaultLocationTag (Variables|Default|GetVaultLocationTag) (Variables|Default|GetSoundEventListener_VividRazorVault_VaultDoor)))
