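; Mansion variant of the VividRazor vault gameplay actor, a child Blueprint that customizes how the vault door looks once opened.
; Its UpdateDoorVisualsOnOpen override runs the parent logic, then sets the first material slot on the door's root mesh to a VaultDoor_B material instance.
; All other vault behavior is inherited, as the construction script only calls the parent and the event stubs are disabled.
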(event EventBeginPlay
  (|CallParentFunction) ; This node is disabled and will not be called.\nDrag off pins to build functionality.
)

(event Collision|EventActorBeginOverlap
  (params (OtherActor :type Actor%20Object%20Reference))
  (|CallParentFunction OtherActor) ; This node is disabled and will not be called.\nDrag off pins to build functionality.
)

(event EventTick
  (params (DeltaSeconds :type Float%20%28single-precision%29))
  (|CallParentFunction DeltaSeconds) ; This node is disabled and will not be called.\nDrag off pins to build functionality.
)

(fn UpdateDoorVisualsOnOpen
  (|CallParentFunction)
  (Rendering|Material|SetMaterial (Variables|VividRazorVaultDoor|GetRootMesh) 0 "/VividRazorVaultGameplay/Assets/Gameplay_VaultDoor_B/Materials/MI_Gameplay_VaultDoor_B_A.MI_Gameplay_VaultDoor_B_A"))

(fn ConstructionScript
  (|CallParentFunction))
