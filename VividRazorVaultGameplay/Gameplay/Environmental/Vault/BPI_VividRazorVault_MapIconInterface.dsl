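; Blueprint interface for controlling a minimap icon on objects in the VividRazorVault gameplay area.
; It declares SetMapIcon, which takes a minimap icon object reference, and SetMapIconVisible, which takes a visibility flag.
; The file holds only these function signatures, with no implementation.
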(fn SetMapIcon
  (params (MiniMapIcon :type Object%20Reference)))

(fn SetMapIconVisible
  (params (IsVisible :type Boolean)))
