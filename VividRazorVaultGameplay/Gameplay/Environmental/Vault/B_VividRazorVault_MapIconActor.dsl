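; Map icon actor for the VividRazor vault, showing the vault's status on the map and minimap through a FortMiniMapComponent.
; On BeginPlay the server sets the component's viewable distance to the larger of the cached map and minimap distances.
; Events set icon visibility, icon texture and an always-show cheat, flushing net dormancy to replicate the changes.
; Icon and cheat changes are authority-only; a separate function sets icon scale.
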(event EventBeginPlay
  ;[0.007568 0.06955 0.07812] BeginPlay: Replicate to the client.
  (|CallParentFunction)
  (Networking|SwitchHasAuthority
    (:Authority
      (MiniMapData|SetMiniMapViewableDistance (Variables|Default|GetFortMiniMapComponent_VaultStatus) (Math|Float|Max(Float) (Variables|/Script/FortniteGame.FortMiniMapComponent_TagsAndDistance|GetCachedMapViewableDistance (Variables|Default|GetFortMiniMapComponent_VaultStatus)) (Variables|/Script/FortniteGame.FortMiniMapComponent_TagsAndDistance|GetCachedMiniMapViewableDistance (Variables|Default|GetFortMiniMapComponent_VaultStatus)))) ; Set the relevancy distance of this object to whichever is greater, the full map or mini map viewable distance.
      (Networking|FlushNetDormancy))))

(event Custom|DeskCurtainContainer_SetAlwaysShowOnMap
  ;[0.4586 0.7292 1] DeskCurtainContainer_SetAlwaysShowOnMap ke *
  ; Cheat to Always Show Them On the Map
  (Networking|SwitchHasAuthority
    (:Authority
      (MiniMapData|SetMiniMapViewableDistance (Variables|Default|GetFortMiniMapComponent_VaultStatus) -1.0)
      (Class|FortMiniMapComponentTagsandDistance|SetViewableDistances (Variables|Default|GetFortMiniMapComponent_VaultStatus) -1.0 -1.0)
      (Networking|FlushNetDormancy))))

(event EventSetMapIconVisible
  (params (IsVisible :type Boolean))
  ;[1 1 1] Set Map Icon Visible
  (MiniMapData|SetMiniMapIndicatorIsVisible (Variables|Default|GetFortMiniMapComponent_VaultStatus) IsVisible)
  (Networking|FlushNetDormancy))

(event EventSetMapIcon
  (params (MiniMapIcon :type Object%20Reference))
  ;[0.02466 0.1126 0.1198] SetMapIcon: Update the map icon.
  (Networking|SwitchHasAuthority
    (:Authority
      (MiniMapData|SetMiniMapIcon (Variables|Default|GetFortMiniMapComponent_VaultStatus) MiniMapIcon)
      (Networking|FlushNetDormancy))))

(fn SetIconScale
  (params (IconScale :type Vector%202D%20Structure :default "(X=1.500000,Y=1.500000)"))
  ;[1 1 1] Set Icon Scale
  (MiniMapData|SetMiniMapIconScale (Variables|Default|GetFortMiniMapComponent_VaultStatus) IconScale))

(fn ConstructionScript)
